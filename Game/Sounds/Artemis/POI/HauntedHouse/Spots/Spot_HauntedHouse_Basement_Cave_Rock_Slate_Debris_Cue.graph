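# Sound cue for falling rock and slate debris spot ambience in the Haunted House basement cave.
# A Random node picks among eight equally weighted debris waves borrowed from the Hollow_Base sound folder and feeds a Modulator at the root.
# It plays quietly on the Athena indoor ambient sound class, using a shared distant-creak one-shot attenuation asset, with bus and submix sends disabled.
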SoundCue /Game/Sounds/Artemis/POI/HauntedHouse/Spots/Spot_HauntedHouse_Basement_Cave_Rock_Slate_Debris_Cue
root=Modulator_0
nodes [
 {id=Modulator_0, children [Random_0]}
 {
  id=Random_0
  children [WavePlayer_0, WavePlayer_1, WavePlayer_2, WavePlayer_3, WavePlayer_4, WavePlayer_11, WavePlayer_15, WavePlayer_16]
  properties {weights [1,1,1,1,1,1,1,1]}
 }
 {id=WavePlayer_0, properties {soundWaveAssetPtr=/Game/Sounds/Artemis/POI/Hollow_Base/Spots/Spot_Cave_Rock_Slate_Debris_12}}
 {id=WavePlayer_1, properties {soundWaveAssetPtr=/Game/Sounds/Artemis/POI/Hollow_Base/Spots/Spot_Cave_Rock_Slate_Debris_04}}
 {id=WavePlayer_11, properties {soundWaveAssetPtr=/Game/Sounds/Artemis/POI/Hollow_Base/Spots/Spot_Cave_Rock_Slate_Debris_08}}
 {id=WavePlayer_15, properties {soundWaveAssetPtr=/Game/Sounds/Artemis/POI/Hollow_Base/Spots/Spot_Cave_Rock_Slate_Debris_10}}
 {id=WavePlayer_16, properties {soundWaveAssetPtr=/Game/Sounds/Artemis/POI/Hollow_Base/Spots/Spot_Cave_Rock_Slate_Debris_11}}
 {id=WavePlayer_2, properties {soundWaveAssetPtr=/Game/Sounds/Artemis/POI/Hollow_Base/Spots/Spot_Cave_Rock_Slate_Debris_05}}
 {id=WavePlayer_3, properties {soundWaveAssetPtr=/Game/Sounds/Artemis/POI/Hollow_Base/Spots/Spot_Cave_Rock_Slate_Debris_06}}
 {id=WavePlayer_4, properties {soundWaveAssetPtr=/Game/Sounds/Artemis/POI/Hollow_Base/Spots/Spot_Cave_Rock_Slate_Debris_07}}
]
properties {
 attenuationSettings=/Game/Sounds/Attenuation/Fort_TileAmbient_OneShot_Distant_Creak_Att
 bEnableBusSends=false
 bEnableSubmixSends=false
 maxDistance=8200
 soundClassObject=&Fort_Ambient_Indoors_Athena
 volumeMultiplier=0.2
 attenuationOverrides {bAttenuate=false}
}
