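# Sound cue for a sand and dust fall spot ambience in the Haunted House basement cave.
# A Random node picks among four equally weighted Cattus Cave sand-dust-fall waves, reused from that POI, and feeds a Modulator at the root.
# It plays at roughly half volume through the indoor ambient sound class, with a shared closer-range one-shot ambient attenuation asset.
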SoundCue /Game/Sounds/Artemis/POI/HauntedHouse/Spots/Spot_HauntedHouse_Basement_Cave_Sand_Dust_Fall_Cue
root=Modulator_0
nodes [
 {id=Modulator_0, children [Random_1]}
 {id=Random_1, children [WavePlayer_5,WavePlayer_6,WavePlayer_7,WavePlayer_8], properties {weights [1,1,1,1]}}
 {id=WavePlayer_5, properties {soundWaveAssetPtr=/Game/Sounds/Artemis/POI/Cattus_Cave/Spots/Spot_Cattus_Cave_Sand_Dust_Fall_01}}
 {id=WavePlayer_6, properties {soundWaveAssetPtr=/Game/Sounds/Artemis/POI/Cattus_Cave/Spots/Spot_Cattus_Cave_Sand_Dust_Fall_02}}
 {id=WavePlayer_7, properties {soundWaveAssetPtr=/Game/Sounds/Artemis/POI/Cattus_Cave/Spots/Spot_Cattus_Cave_Sand_Dust_Fall_03}}
 {id=WavePlayer_8, properties {soundWaveAssetPtr=/Game/Sounds/Artemis/POI/Cattus_Cave/Spots/Spot_Cattus_Cave_Sand_Dust_Fall_04}}
]
properties {
 attenuationSettings=/Game/Sounds/Attenuation/Fort_TileAmbient_OneShot_Att_Closer
 maxDistance=8500
 soundClassObject=&Fort_Ambient_Indoors_Athena
 volumeMultiplier=0.501
 attenuationOverrides {bAttenuate=false}
}
