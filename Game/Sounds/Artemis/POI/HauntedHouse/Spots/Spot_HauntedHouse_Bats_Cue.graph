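# Sound cue for bat spot ambience at the Haunted House POI, reusing five bat waves from Cattus Cave.
# An equal-weight Random node picks one wave and feeds a Modulator at default settings for slight pitch and volume variation.
# It uses a shared closer one-shot tile-ambient attenuation asset and an interior-space ambient ducking sound class.
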SoundCue /Game/Sounds/Artemis/POI/HauntedHouse/Spots/Spot_HauntedHouse_Bats_Cue
root=Modulator_0
nodes [
 {id=Modulator_0, children [Random_0]}
 {id=Random_0, children [WavePlayer_0,WavePlayer_1,WavePlayer_2,WavePlayer_3,WavePlayer_4], properties {weights [1,1,1,1,1]}}
 {id=WavePlayer_0, properties {soundWaveAssetPtr=/Game/Sounds/Artemis/POI/Cattus_Cave/Spots/Spot_Cattus_Cave_Bat_05}}
 {id=WavePlayer_1, properties {soundWaveAssetPtr=/Game/Sounds/Artemis/POI/Cattus_Cave/Spots/Spot_Cattus_Cave_Bat_02}}
 {id=WavePlayer_2, properties {soundWaveAssetPtr=/Game/Sounds/Artemis/POI/Cattus_Cave/Spots/Spot_Cattus_Cave_Bat_01}}
 {id=WavePlayer_3, properties {soundWaveAssetPtr=/Game/Sounds/Artemis/POI/Cattus_Cave/Spots/Spot_Cattus_Cave_Bat_03}}
 {id=WavePlayer_4, properties {soundWaveAssetPtr=/Game/Sounds/Artemis/POI/Cattus_Cave/Spots/Spot_Cattus_Cave_Bat_04}}
]
properties {
 attenuationSettings=/Game/Sounds/Attenuation/Fort_TileAmbient_OneShot_Att_Closer
 maxDistance=8500
 soundClassObject=/Game/Sounds/FortSoundClassesAndModes/Fort_Ambient_InteriorSpace_Duck_TileMix_PSM
 volumeMultiplier=1
 attenuationOverrides {bAttenuate=false}
}
