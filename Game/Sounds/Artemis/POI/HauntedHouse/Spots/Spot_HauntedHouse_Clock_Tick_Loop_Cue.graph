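# Haunted house clock SoundCue in the indoor ambient sound class, combining a ticking loop with a delayed chime for a spot sound.
# A looping tick and a single chime held back by a random 30 to 60 second delay are each boosted by a mixer, then given separate attenuation assets borrowed from the ACabin clock sounds.
# A final mixer feeds both branches identically into every quality level.
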SoundCue /Game/Sounds/Artemis/POI/HauntedHouse/Spots/Spot_HauntedHouse_Clock_Tick_Loop_Cue
root=QualityLevel_0
nodes [
 {
  id=Attenuation_0
  children [Mixer_1]
  properties {attenuationSettings=/Game/Sounds/Artemis/POI/ACabin/Spots/Spot_ACabin_Clock_Rining_Sequence_Att}
 }
 {
  id=Attenuation_1
  children [Mixer_0]
  properties {attenuationSettings=/Game/Sounds/Artemis/POI/ACabin/Spots/Spot_ACabin_Clock_Ticking_Loop_Att}
 }
 {id=Delay_0, children [WavePlayer_1], properties {delayMax=60, delayMin=30}}
 {id=Mixer_0, children [WavePlayer_0], properties {inputVolume [2]}}
 {id=Mixer_1, children [Delay_0], properties {inputVolume [2]}}
 {id=Mixer_2, children [Attenuation_1, Attenuation_0], properties {inputVolume [1,1]}}
 {id=QualityLevel_0, children [Mixer_2,Mixer_2,Mixer_2,Mixer_2]}
 {
  id=WavePlayer_0
  properties {bLooping=true, soundWaveAssetPtr=&/Spot_HauntedHouse_Clock_TickLoop_01}
 }
 {id=WavePlayer_1, properties {soundWaveAssetPtr=&/Spot_HauntedHouse_Clock_Chime_01}}
]
properties {
 bEnableBusSends=false
 maxDistance=1150
 soundClassObject=&Fort_Ambient_Indoors_Athena
 volumeMultiplier=2.506
 preEffectBusSends [{}]
 soundSubmixSends [{minSendLevel=0.4, sendLevel=0, soundSubmix=&Fortnite_Reverb_Distant}]
}
