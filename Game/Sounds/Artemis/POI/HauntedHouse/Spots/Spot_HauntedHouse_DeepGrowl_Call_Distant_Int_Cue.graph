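# Sound cue for distant deep growl calls used as interior ambient spot one-shots at the Haunted House POI, built from reused UFO crash site and Tower Aftermath alien call waves.
# A quality-level node sends its first two levels to a large equal-weight random pool and its last two to a four-wave subset.
# Both pools get random pitch and volume modulation, then play through indoor ambient sound class and wide one-shot attenuation.
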SoundCue /Game/Sounds/Artemis/POI/HauntedHouse/Spots/Spot_HauntedHouse_DeepGrowl_Call_Distant_Int_Cue
root=QualityLevel_0
nodes [
 {id=Mixer_0, children [WavePlayer_5], properties {inputVolume [0.7]}}
 {id=Modulator_1, children [Random_1], properties {pitchMax=1.1, pitchMin=0.8, volumeMax=1.1, volumeMin=0.9}}
 {id=Modulator_2, children [Random_3], properties {pitchMax=1.1, pitchMin=0.8, volumeMax=1.1, volumeMin=0.9}}
 {id=QualityLevel_0, children [Modulator_1,Modulator_1,Modulator_2,Modulator_2]}
 {
  id=Random_1
  children [
   WavePlayer_1
   WavePlayer_2
   WavePlayer_3
   WavePlayer_4
   Mixer_0
   WavePlayer_6
   WavePlayer_7
   WavePlayer_8
   WavePlayer_9
   WavePlayer_10
   WavePlayer_23
   WavePlayer_24
   WavePlayer_25
   WavePlayer_26
   WavePlayer_27
   WavePlayer_28
   WavePlayer_29
   WavePlayer_30
   WavePlayer_31
   WavePlayer_66
   WavePlayer_65
   WavePlayer_64
   WavePlayer_63
   WavePlayer_62
  ]
  properties {weights [1,1,1,1,1,1,1,1,1,1,1,1,1,1,1,1,1,1,1,1,1,1,1,1]}
 }
 {id=Random_3, children [WavePlayer_33, WavePlayer_32, WavePlayer_38, WavePlayer_39], properties {weights [1,1,1,1]}}
 {
  id=WavePlayer_1
  properties {soundWaveAssetPtr=/Game/Sounds/Apollo/Biomes/UFO_Crash_Sites/AmbientOneshots/AMB_UFO_Crashsite_Alien_Call_Spot_19}
 }
 {
  id=WavePlayer_10
  properties {soundWaveAssetPtr=/Game/Sounds/Apollo/Biomes/UFO_Crash_Sites/AmbientOneshots/AMB_UFO_Crashsite_Alien_Call_Spot_18}
 }
 {
  id=WavePlayer_2
  properties {soundWaveAssetPtr=/Game/Sounds/Apollo/Biomes/UFO_Crash_Sites/AmbientOneshots/AMB_UFO_Crashsite_Alien_Call_Spot_03}
 }
 {
  id=WavePlayer_23
  properties {soundWaveAssetPtr=/Game/Sounds/Apollo/Biomes/UFO_Crash_Sites/AmbientOneshots/AMB_UFO_Crashsite_Alien_Call_Spot_20}
 }
 {
  id=WavePlayer_24
  properties {soundWaveAssetPtr=/Game/Sounds/Apollo/Biomes/UFO_Crash_Sites/AmbientOneshots/AMB_UFO_Crashsite_Alien_Call_Spot_01}
 }
 {
  id=WavePlayer_25
  properties {soundWaveAssetPtr=/Game/Sounds/Apollo/Biomes/UFO_Crash_Sites/AmbientOneshots/AMB_UFO_Crashsite_Alien_Call_Spot_02}
 }
 {
  id=WavePlayer_26
  properties {soundWaveAssetPtr=/Game/Sounds/Apollo/Biomes/UFO_Crash_Sites/AmbientOneshots/AMB_UFO_Crashsite_Alien_Call_Spot_04}
 }
 {
  id=WavePlayer_27
  properties {soundWaveAssetPtr=/Game/Sounds/Apollo/Biomes/UFO_Crash_Sites/AmbientOneshots/AMB_UFO_Crashsite_Alien_Call_Spot_05}
 }
 {
  id=WavePlayer_28
  properties {soundWaveAssetPtr=/Game/Sounds/Apollo/Biomes/UFO_Crash_Sites/AmbientOneshots/AMB_UFO_Crashsite_Alien_Call_Spot_06}
 }
 {
  id=WavePlayer_29
  properties {soundWaveAssetPtr=/Game/Sounds/Apollo/Biomes/UFO_Crash_Sites/AmbientOneshots/AMB_UFO_Crashsite_Alien_Call_Spot_07}
 }
 {
  id=WavePlayer_3
  properties {soundWaveAssetPtr=/Game/Sounds/Apollo/Biomes/UFO_Crash_Sites/AmbientOneshots/AMB_UFO_Crashsite_Alien_Call_Spot_10}
 }
 {
  id=WavePlayer_30
  properties {soundWaveAssetPtr=/Game/Sounds/Apollo/Biomes/UFO_Crash_Sites/AmbientOneshots/AMB_UFO_Crashsite_Alien_Call_Spot_08}
 }
 {
  id=WavePlayer_31
  properties {soundWaveAssetPtr=/Game/Sounds/Apollo/Biomes/UFO_Crash_Sites/AmbientOneshots/AMB_UFO_Crashsite_Alien_Call_Spot_09}
 }
 {
  id=WavePlayer_32
  properties {soundWaveAssetPtr=&AMB_UFO_Crashsite_Alien_Call_Spot_03}
 }
 {
  id=WavePlayer_33
  properties {soundWaveAssetPtr=&AMB_UFO_Crashsite_Alien_Call_Spot_19}
 }
 {
  id=WavePlayer_38
  properties {soundWaveAssetPtr=/Game/Sounds/Apollo/Biomes/UFO_Crash_Sites/AmbientOneshots/AMB_UFO_Crashsite_Alien_Call_Spot_14}
 }
 {
  id=WavePlayer_39
  properties {soundWaveAssetPtr=&AMB_UFO_Crashsite_Alien_Call_Spot_20}
 }
 {
  id=WavePlayer_4
  properties {soundWaveAssetPtr=/Game/Sounds/Apollo/Biomes/UFO_Crash_Sites/AmbientOneshots/AMB_UFO_Crashsite_Alien_Call_Spot_11}
 }
 {
  id=WavePlayer_5
  properties {soundWaveAssetPtr=/Game/Sounds/Apollo/Biomes/UFO_Crash_Sites/AmbientOneshots/AMB_UFO_Crashsite_Alien_Call_Spot_12}
 }
 {
  id=WavePlayer_6
  properties {soundWaveAssetPtr=/Game/Sounds/Apollo/Biomes/UFO_Crash_Sites/AmbientOneshots/AMB_UFO_Crashsite_Alien_Call_Spot_13}
 }
 {id=WavePlayer_62, properties {soundWaveAssetPtr=/Game/Sounds/Apollo/Biomes/Tower_Aftermath/AMB_Tower_AM_Alien_Call_Distant_07}}
 {id=WavePlayer_63, properties {soundWaveAssetPtr=/Game/Sounds/Apollo/Biomes/Tower_Aftermath/AMB_Tower_AM_Alien_Call_Distant_05}}
 {id=WavePlayer_64, properties {soundWaveAssetPtr=/Game/Sounds/Apollo/Biomes/Tower_Aftermath/AMB_Tower_AM_Alien_Call_Distant_03}}
 {id=WavePlayer_65, properties {soundWaveAssetPtr=/Game/Sounds/Apollo/Biomes/Tower_Aftermath/AMB_Tower_AM_Alien_Call_Distant_01}}
 {id=WavePlayer_66, properties {soundWaveAssetPtr=/Game/Sounds/Apollo/Biomes/Tower_Aftermath/AMB_Tower_AM_Alien_Call_Distant_12}}
 {
  id=WavePlayer_7
  properties {soundWaveAssetPtr=&AMB_UFO_Crashsite_Alien_Call_Spot_14}
 }
 {
  id=WavePlayer_8
  properties {soundWaveAssetPtr=/Game/Sounds/Apollo/Biomes/UFO_Crash_Sites/AmbientOneshots/AMB_UFO_Crashsite_Alien_Call_Spot_15}
 }
 {
  id=WavePlayer_9
  properties {soundWaveAssetPtr=/Game/Sounds/Apollo/Biomes/UFO_Crash_Sites/AmbientOneshots/AMB_UFO_Crashsite_Alien_Call_Spot_16}
 }
]
properties {
 attenuationSettings=/Game/Sounds/Attenuation/Fort_TileAmbient_OneShot_Wider_Att
 maxDistance=8000
 soundClassObject=&Fort_Ambient_Indoors_Athena
 volumeMultiplier=0.501
 preEffectBusSends [{}]
}
