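# Sound Cue for distant animal calls used as a spot ambience at the Haunted House point of interest.
# A Random node picks one of eight equally weighted cave creature call waves, reused from Starlight S3 audio, and feeds it through a Modulator at the root.
# It uses a distant-creature one-shot attenuation asset and an ambient interior-space ducking sound class.
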SoundCue /Game/Sounds/Artemis/POI/HauntedHouse/Spots/Spot_HauntedHouse_Distant_Animal_Cue
root=Modulator_1
nodes [
 {id=Modulator_1, children [Random_5]}
 {
  id=Random_5
  children [WavePlayer_25, WavePlayer_26, WavePlayer_27, WavePlayer_28, WavePlayer_29, WavePlayer_30, WavePlayer_31, WavePlayer_32]
  properties {weights [1,1,1,1,1,1,1,1]}
 }
 {id=WavePlayer_25, properties {soundWaveAssetPtr=/Game/Sounds/Fort_Starlight/S3/SoundWaves/Cave_Creature_Call_Random_08}}
 {id=WavePlayer_26, properties {soundWaveAssetPtr=/Game/Sounds/Fort_Starlight/S3/SoundWaves/Cave_Creature_Call_Random_01}}
 {id=WavePlayer_27, properties {soundWaveAssetPtr=/Game/Sounds/Fort_Starlight/S3/SoundWaves/Cave_Creature_Call_Random_02}}
 {id=WavePlayer_28, properties {soundWaveAssetPtr=/Game/Sounds/Fort_Starlight/S3/SoundWaves/Cave_Creature_Call_Random_03}}
 {id=WavePlayer_29, properties {soundWaveAssetPtr=/Game/Sounds/Fort_Starlight/S3/SoundWaves/Cave_Creature_Call_Random_04}}
 {id=WavePlayer_30, properties {soundWaveAssetPtr=/Game/Sounds/Fort_Starlight/S3/SoundWaves/Cave_Creature_Call_Random_05}}
 {id=WavePlayer_31, properties {soundWaveAssetPtr=/Game/Sounds/Fort_Starlight/S3/SoundWaves/Cave_Creature_Call_Random_06}}
 {id=WavePlayer_32, properties {soundWaveAssetPtr=/Game/Sounds/Fort_Starlight/S3/SoundWaves/Cave_Creature_Call_Random_07}}
]
properties {
 attenuationSettings=/Game/Sounds/Attenuation/Fort_TileAmbient_OneShot_Distant_Creature_Att
 maxDistance=8500
 soundClassObject=/Game/Sounds/FortSoundClassesAndModes/Fort_Ambient_InteriorSpace_Duck_TileMix_PSM
 volumeMultiplier=1
 attenuationOverrides {bAttenuate=false}
}
