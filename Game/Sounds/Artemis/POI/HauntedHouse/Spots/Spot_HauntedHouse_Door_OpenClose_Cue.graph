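# Sound cue for haunted house ambient spot sounds of a door opening or closing, at the Artemis point of interest.
# A random node picks with equal weight among two door-open waves and one door-close wave, so each play is either action.
# All four quality-level inputs share that random node.
# It uses a reduced volume, a tile ambient one-shot attenuation and the ambient one-shots sound class, with bus sends disabled.
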SoundCue /Game/Sounds/Artemis/POI/HauntedHouse/Spots/Spot_HauntedHouse_Door_OpenClose_Cue
root=QualityLevel_0
nodes [
 {id=QualityLevel_0, children [Random_0,Random_0,Random_0,Random_0]}
 {id=Random_0, children [WavePlayer_0,WavePlayer_1,WavePlayer_2], properties {weights [1,1,1]}}
 {id=WavePlayer_0, properties {soundWaveAssetPtr=&/Spot_HauntedHouse_DoorOpen_01}}
 {id=WavePlayer_1, properties {soundWaveAssetPtr=&/Spot_HauntedHouse_DoorOpen_02}}
 {id=WavePlayer_2, properties {soundWaveAssetPtr=&/Spot_HauntedHouse_DoorClose_01}}
]
properties {
 attenuationSettings=/Game/Sounds/Attenuation/Fort_TileAmbient_OneShot_Wind_Less_Spread_Att
 bEnableBusSends=false
 maxDistance=8500
 soundClassObject=&Fort_Ambient_OneShots
 volumeMultiplier=0.376
 preEffectBusSends [{}]
}
