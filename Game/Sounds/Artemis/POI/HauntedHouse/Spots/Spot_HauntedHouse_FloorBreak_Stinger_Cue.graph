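# Sound cue for the haunted house POI stinger, named for the floor break, on the indoor ambient sound class.
# A Mixer layers the floor-fall stinger with the enter-basement stinger, which a fixed half-second Delay holds back.
# An Attenuation node overrides the ambient attenuation with a custom curve held at full volume, and all four quality levels share this chain.
# A concurrency override limits the cue to one instance.
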SoundCue /Game/Sounds/Artemis/POI/HauntedHouse/Spots/Spot_HauntedHouse_FloorBreak_Stinger_Cue
root=QualityLevel_1
nodes [
 {
  id=Attenuation_2
  children [Mixer_0]
  properties {
   attenuationSettings=&Fort_Ambient_Athena_Att
   attenuationOverrides {
    distanceAlgorithm=Custom
    falloffDistance=0
    attenuationShapeExtents {x=0}
    customAttenuationCurve {editorCurveData {keys [{interpMode=RCIM_Cubic, tangentMode=RCTM_Break, time=0, value=1}, <at 1 1>]}}
   }
  }
 }
 {id=Delay_1, children [WavePlayer_0], properties {delayMax=0.5, delayMin=0.5}}
 {id=Mixer_0, children [WavePlayer_5,Delay_1], properties {inputVolume [1,1]}}
 {id=QualityLevel_1, children [Attenuation_2, Attenuation_2, Attenuation_2, Attenuation_2]}
 {id=WavePlayer_0, properties {soundWaveAssetPtr=/Game/Sounds/Artemis/POI/HauntedHouse/HauntedHouse_Stinger_EnterBasement_01}}
 {id=WavePlayer_5, properties {soundWaveAssetPtr=/Game/Sounds/Artemis/POI/HauntedHouse/HauntedHouse_Stinger_FloorFall_01}}
]
properties {
 bOverrideConcurrency=true
 maxDistance=10000
 soundClassObject=&Fort_Ambient_Indoors_Athena
 volumeMultiplier=1.002
 concurrencyOverrides {maxCount=1, volumeScaleMode=Distance}
}
