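# Sound cue for the wooden floor-break spot effect in the Haunted House POI, classed as Athena ambient audio.
# A mixer layers a wood break and two stingers with a falling-debris wave that is delayed half a second and pitched down.
# An attenuation override makes the mix non-spatialized but still sets a distance falloff.
# All quality levels share that chain, and concurrency allows one instance at a time.
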SoundCue /Game/Sounds/Artemis/POI/HauntedHouse/Spots/Spot_HauntedHouse_FloorBreak_Wood_Cue
root=QualityLevel_1
nodes [
 {
  id=Attenuation_2
  children [Mixer_0]
  properties {
   attenuationSettings=&Fort_Ambient_Athena_Att
   bOverrideAttenuation=true
   attenuationOverrides {
    bSpatialize=false
    falloffDistance=2000
    attenuationShapeExtents {x=1000}
    customAttenuationCurve {editorCurveData {keys [{interpMode=RCIM_Cubic, tangentMode=RCTM_Break, time=0, value=1}, <at 1 1>]}}
   }
  }
 }
 {id=Delay_0, children [Modulator_0], properties {delayMax=0.5, delayMin=0.5}}
 {id=Mixer_0, children [WavePlayer_0,WavePlayer_1,WavePlayer_2,Delay_0], properties {inputVolume [1.3,1,0.5,1.3]}}
 {id=Modulator_0, children [WavePlayer_3], properties {pitchMax=0.9, pitchMin=0.9}}
 {id=QualityLevel_1, children [Attenuation_2, Attenuation_2, Attenuation_2, Attenuation_2]}
 {id=WavePlayer_0, properties {soundWaveAssetPtr=&/Spot_HauntedHouse_FloorBreak_Wood_01}}
 {id=WavePlayer_1, properties {soundWaveAssetPtr=/Game/Sounds/Artemis/POI/HauntedHouse/HauntedHouse_Stinger_FloorBreak_01}}
 {id=WavePlayer_2, properties {soundWaveAssetPtr=&/Spot_HauntedHouse_FloorBreak_Stinger_01}}
 {id=WavePlayer_3, properties {soundWaveAssetPtr=/Game/Sounds/Artemis/POI/HauntedHouse/HauntedHouse_Stinger_FloorFall_Debris_01}}
]
properties {
 bOverrideConcurrency=true
 maxDistance=3000
 soundClassObject=&Fort_Ambient_Athena
 volumeMultiplier=1.153
 concurrencyOverrides {maxCount=1, volumeScaleMode=Distance}
}
