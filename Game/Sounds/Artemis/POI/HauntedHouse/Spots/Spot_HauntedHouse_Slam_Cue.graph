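# Sound cue for a slam one-shot spot sound at the Haunted House POI, classed as an ambient one-shot.
# A Random node picks among four equally weighted slam waves, passing through a QualityLevel node whose four inputs share that same Random, then a root Modulator.
# Playback uses a mono tile-ambient one-shot attenuation asset at reduced volume, with a send to the distant reverb submix.
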SoundCue /Game/Sounds/Artemis/POI/HauntedHouse/Spots/Spot_HauntedHouse_Slam_Cue
root=Modulator_0
nodes [
 {id=Modulator_0, children [QualityLevel_0]}
 {id=QualityLevel_0, children [Random_1,Random_1,Random_1,Random_1]}
 {id=Random_1, children [WavePlayer_3,WavePlayer_4,WavePlayer_5,WavePlayer_6], properties {weights [1,1,1,1]}}
 {id=WavePlayer_3, properties {soundWaveAssetPtr=&/Spot_HauntedHouse_Slam_01}}
 {id=WavePlayer_4, properties {soundWaveAssetPtr=&/Spot_HauntedHouse_Slam_02}}
 {id=WavePlayer_5, properties {soundWaveAssetPtr=&/Spot_HauntedHouse_Slam_03}}
 {id=WavePlayer_6, properties {soundWaveAssetPtr=&/Spot_HauntedHouse_Slam_04}}
]
properties {
 attenuationSettings=/Game/Sounds/Attenuation/Fort_TileAmbient_OneShot_Maxium_Mono_Att
 bEnableBusSends=false
 maxDistance=8500
 soundClassObject=&Fort_Ambient_OneShots
 volumeMultiplier=0.501
 preEffectBusSends [{}]
 soundSubmixSends [{maxSendLevel=0.6, minSendLevel=0.2, sendLevel=0, soundSubmix=&Fortnite_Reverb_Distant}]
}
