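# Sound cue for a wolf ambient one-shot spot in the Haunted House POI, apparently the interior variant.
# A Random node picks among five equally weighted wolf waves, and a quality-level node routes all four of its inputs to that same Random.
# It plays at reduced volume through the ambient one-shot sound class and a wide air-absorption attenuation, with a submix send configured for distant reverb.
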SoundCue /Game/Sounds/Artemis/POI/HauntedHouse/Spots/Spot_HauntedHouse_Wolf_Int_Cue
root=QualityLevel_0
nodes [
 {id=QualityLevel_0, children [Random_0,Random_0,Random_0,Random_0]}
 {id=Random_0, children [WavePlayer_0,WavePlayer_1,WavePlayer_2,WavePlayer_8,WavePlayer_9], properties {weights [1,1,1,1,1]}}
 {id=WavePlayer_0, properties {soundWaveAssetPtr=&/Spot_HauntedHouse_Wolf_01}}
 {id=WavePlayer_1, properties {soundWaveAssetPtr=&/Spot_HauntedHouse_Wolf_02}}
 {id=WavePlayer_2, properties {soundWaveAssetPtr=&/Spot_HauntedHouse_Wolf_03}}
 {id=WavePlayer_8, properties {soundWaveAssetPtr=&/Spot_HauntedHouse_Wolf_04}}
 {id=WavePlayer_9, properties {soundWaveAssetPtr=&/Spot_HauntedHouse_Wolf_05}}
]
properties {
 attenuationSettings=/Game/Sounds/Attenuation/Fort_TileAmbient_OneShot_Wider_AirAbsorbtion_Att
 bEnableBusSends=false
 maxDistance=8000
 soundClassObject=&Fort_Ambient_OneShots
 volumeMultiplier=0.501
 busSends [{audioBus=null, minSendLevel=0, sendLevel=0, soundSourceBus=null}]
 preEffectBusSends [{}]
 soundSubmixSends [{maxSendLevel=0.6, minSendLevel=0.2, sendLevel=0, soundSubmix=&Fortnite_Reverb_Distant}]
}
