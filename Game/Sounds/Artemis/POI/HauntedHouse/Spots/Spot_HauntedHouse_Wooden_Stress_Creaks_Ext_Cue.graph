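# Outdoor ambient spot cue of wooden stress creaks for the Haunted House POI.
# A looping house-creak bed mixes with half-volume random bridge creaks, retriggered after random delays.
# The first two quality levels use this mix, while the other two play only the bed.
# A start delay and pitch/volume modulation feed two nested looping envelopes, shaping pitch and volume over different periods.
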SoundCue /Game/Sounds/Artemis/POI/HauntedHouse/Spots/Spot_HauntedHouse_Wooden_Stress_Creaks_Ext_Cue
root=Enveloper_2
nodes [
 {id=Delay_0, children [Random_0], properties {delayMax=25, delayMin=3}}
 {id=Delay_1, children [Modulator_0], properties {delayMax=2, delayMin=0.2}}
 {
  id=Enveloper_1
  children [Delay_1]
  properties {
   bLoop=true
   bLoopIndefinitely=true
   loopEnd=30
   pitchCurve {editorCurveData {keys [{interpMode=RCIM_Cubic, time=0}, <at 15.9468 0.9181 RCIM_Cubic>, {interpMode=RCIM_Cubic, time=30}]}}
   volumeCurve {editorCurveData {keys [<at 0 1>]}}
  }
 }
 {
  id=Enveloper_2
  children [Enveloper_1]
  properties {
   bLoop=true
   bLoopIndefinitely=true
   loopEnd=45
   pitchCurve {editorCurveData {keys [{time=0}]}}
   volumeCurve {editorCurveData {keys [
      {arriveTangent=-0.01, interpMode=RCIM_Cubic, time=0, value=1}
      {arriveTangent=-0.01, interpMode=RCIM_Cubic, leaveTangent=-0.01, time=10, value=0.6}
      {arriveTangent=0, interpMode=RCIM_Cubic, leaveTangent=0, tangentMode=RCTM_User, time=20, value=0.8}
      {arriveTangent=0.008, interpMode=RCIM_Cubic, leaveTangent=0.008, time=30, value=0.5}
      <at 45 1 RCIM_Cubic>
     ]}}
  }
 }
 {id=Looping_0, children [Delay_0]}
 {id=Mixer_0, children [WavePlayer_13, Looping_0], properties {inputVolume [1,0.5]}}
 {id=Modulator_0, children [QualityLevel_0], properties {pitchMin=0.8, volumeMin=0.8}}
 {id=QualityLevel_0, children [Mixer_0, Mixer_0, WavePlayer_13, WavePlayer_13]}
 {
  id=Random_0
  children [WavePlayer_0,WavePlayer_1,WavePlayer_2,WavePlayer_3,WavePlayer_4,WavePlayer_5,WavePlayer_6,WavePlayer_7,WavePlayer_8]
  properties {weights [1,1,1,1,1,1,1,1,1]}
 }
 {
  id=WavePlayer_0
  properties {soundWaveAssetPtr=/Game/Sounds/Artemis/POI/Normal_Bridges/Bridges_Wooden_Corridor/Spots/Spot_Bridge_Wooden_Corridor_Creaks_09}
 }
 {
  id=WavePlayer_1
  properties {soundWaveAssetPtr=/Game/Sounds/Artemis/POI/Normal_Bridges/Bridges_Wooden_Corridor/Spots/Spot_Bridge_Wooden_Corridor_Creaks_01}
 }
 {
  id=WavePlayer_13
  properties {bLooping=true, soundWaveAssetPtr=/Game/Sounds/Artemis/POI/HauntedHouse/HauntedHouse_AmbientLoop_Exterior_HouseCreak_01}
 }
 {
  id=WavePlayer_2
  properties {soundWaveAssetPtr=/Game/Sounds/Artemis/POI/Normal_Bridges/Bridges_Wooden_Corridor/Spots/Spot_Bridge_Wooden_Corridor_Creaks_02}
 }
 {
  id=WavePlayer_3
  properties {soundWaveAssetPtr=/Game/Sounds/Artemis/POI/Normal_Bridges/Bridges_Wooden_Corridor/Spots/Spot_Bridge_Wooden_Corridor_Creaks_03}
 }
 {
  id=WavePlayer_4
  properties {soundWaveAssetPtr=/Game/Sounds/Artemis/POI/Normal_Bridges/Bridges_Wooden_Corridor/Spots/Spot_Bridge_Wooden_Corridor_Creaks_04}
 }
 {
  id=WavePlayer_5
  properties {soundWaveAssetPtr=/Game/Sounds/Artemis/POI/Normal_Bridges/Bridges_Wooden_Corridor/Spots/Spot_Bridge_Wooden_Corridor_Creaks_05}
 }
 {
  id=WavePlayer_6
  properties {soundWaveAssetPtr=/Game/Sounds/Artemis/POI/Normal_Bridges/Bridges_Wooden_Corridor/Spots/Spot_Bridge_Wooden_Corridor_Creaks_06}
 }
 {
  id=WavePlayer_7
  properties {soundWaveAssetPtr=/Game/Sounds/Artemis/POI/Normal_Bridges/Bridges_Wooden_Corridor/Spots/Spot_Bridge_Wooden_Corridor_Creaks_07}
 }
 {
  id=WavePlayer_8
  properties {soundWaveAssetPtr=/Game/Sounds/Artemis/POI/Normal_Bridges/Bridges_Wooden_Corridor/Spots/Spot_Bridge_Wooden_Corridor_Creaks_08}
 }
]
properties {
 attenuationSettings=/Game/Sounds/Attenuation/Spot_Ambient_Presence_Spread_Focus_Distant_Att
 maxDistance=7000
 soundClassObject=&Fort_Ambient_Outdoors_Athena
 volumeMultiplier=0.501
}
